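; BP_BattleStage_PlatformBase is a battle stage platform Blueprint that lowers once when its LowerMe event is called.
; LowerMe checks for a valid platform mesh, logging if none exists, and uses bIsLowered to ignore repeat calls.
; A timeline curve, scaled by a fixed distance, drives the mesh's relative Z while the lower sound plays at the root's location.
; Nothing here resets the lowered state, and the construction script is empty.
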(event Custom|LowerMe
  (Variables|Default|GetPlatformMesh
    (:then
      (if (not (Variables|Default|GetbIsLowered))
        (Variables|Default|SetbIsLowered true)
        (Utilities|FlowControl|Sequence
          (:then_0
            (bind (_direction _platform_movedown) (Timeline|Timeline Play)
              (:Update
                (Transformation|SetRelativeLocation (Variables|Default|GetPlatformMesh) (Math|Vector|MakeVector 0.0 0.0 (* _platform_movedown 147.0)) false false))))
          (:then_1
            (Audio|PlaySoundatLocation (Variables|Default|GetLowerSound) (Transformation|GetWorldLocation (Variables|Default|GetDefaultSceneRoot)))))))
    (:else
      (Fortnite|FortShippingLog self "Platform does not have a vaild mesh" true))))

(fn ConstructionScript)
