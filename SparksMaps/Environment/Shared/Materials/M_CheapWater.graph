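# Cheap water surface material for SparksMaps environments, with switchable color gradient, caustics and sparkles.
# Two oppositely panning normal samples blend into the surface normal, and their multiplied red channels, raised to a power, form sparkles.
# Color is flat or a view-angle and depth gradient, while caustics multiply two scrolling BumpOffset samples for fake depth.
# Quality switches blend in a reflection cubemap except on High and Epic, and toggle constant values.
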Material /SparksMaps/Environment/Shared/Materials/M_CheapWater
graph {
 properties {bAutomaticallySetUsageInEditor=false, usedShadingModels=Default Lit}
 nodes [
  n1 = TextureCoordinate() {uTiling=1.65, vTiling=1.65}
  n2 = ScalarParameter() {defaultValue=15, group=Normal, parameterName=Normal UV Multiplier}
  n3 = TextureObjectParameter() {
   group=Normal
   parameterName=Normal Texture
   samplerType=SAMPLERTYPE_Normal
   sortPriority=1
   texture=/Game/Global/Textures/Default/Noise/T_TilingClouds_N
  }
  n4 = TextureSample(UVs=Panner(Coordinate=Multiply(A=@n1, B=@n2) {constB=5}) {speedX=0.05, speedY=0.05}, Tex=@n3) {samplerType=SAMPLERTYPE_Normal, texture=/Game/Gadgets/Assets/Telephone_Booth/Textures/T_BlindsNormalNoise_N}
  Multiply(A=Power(Base=@n4.R) {constExponent=5}) {constB=1000}
  n5 = TextureCoordinate() {uTiling=2, vTiling=2}
  n6 = TextureSample(UVs=Panner(Coordinate=Multiply(A=@n5, B=@n2) {constB=5}) {speedX=-0.05, speedY=-0.08}, Tex=@n3) {samplerType=SAMPLERTYPE_Normal, texture=&T_BlindsNormalNoise_N}
  Multiply(A=Power(Base=@n6.R) {constExponent=4}) {constB=1000}
  n7 = TextureCoordinate()
  n8 = Multiply(A=Time(), B=ScalarParameter() {defaultValue=0.004, group=Caustics, parameterName=Caustics 1 Speed, sortPriority=4}) {constB=0.03}
  n9 = ScalarParameter() {defaultValue=-50, group=Caustics, parameterName=Caustics Depth, sortPriority=3}
  n10 = BumpOffset(
   Coordinate=Multiply(A=Add(A=@n7, B=@n8), B=ScalarParameter() {defaultValue=12, group=Caustics, parameterName=Caustics 1 Scale, sortPriority=5})
   Height=@n9
  )
  n11 = TextureCoordinate()
  n12 = Multiply(A=Time(), B=ScalarParameter() {defaultValue=0.005, group=Caustics, parameterName=Caustics 2 Speed, sortPriority=6}) {constB=0.02}
  n13 = ScalarParameter() {defaultValue=20, group=Caustics, parameterName=Caustics 2 Scale, sortPriority=7}
  n14 = Constant2Vector() {g=1, r=0.5}
  n15 = BumpOffset(Coordinate=Multiply(A=Subtract(A=@n11, B=@n12), B=Multiply(A=@n13, B=@n14)), Height=@n9)
  NamedRerouteDeclaration(None=Multiply(
    A=TextureSample(UVs=@n10) {samplerType=SAMPLERTYPE_LinearGrayscale, texture=/FortWater/Materials/WaterSurface/Mobile/Textures/T_MobileWater_Caustics}.R
    B=TextureSample(UVs=@n15) {samplerType=SAMPLERTYPE_LinearGrayscale, texture=&T_MobileWater_Caustics}.R
   )) {name=Caustics, nodeColor=<rgba 1 0.9141 0 1>}
  n16 = VectorParameter() {group=Gradient, parameterName=Gradient Color 1, sortPriority=2, defaultValue=<rgba 0.0176 0.0976 0.2159 1>}
  n17 = VectorParameter() {group=Gradient, parameterName=Gradient Color 2, sortPriority=3, defaultValue=<rgba 0.0144 0.1046 0.2396 1>}
  n18 = Power(Base=DotProduct(A=VertexNormalWS(), B=CameraVectorWS())) {constExponent=1.5}
  n19 = OneMinus(None=Divide(A=PixelDepth().R) {constB=4098})
  n20 = LinearInterpolate(A=@n16.RGB, B=@n17.RGB, Alpha=Saturate(None=Add(A=@n18, B=@n19)))
  n21 = VectorParameter() {group=Flat Color, parameterName=Flat Color, sortPriority=1, defaultValue=<rgba 0.0176 0.0176 0.0176 1>}
  NamedRerouteDeclaration(None=StaticSwitchParameter(True=NamedRerouteDeclaration(None=@n20) {name=Color Gradient, nodeColor=<rgba 0.0026 0.0135 0.026 1>}, False=@n21.RGB) {group=Gradient, parameterName=Color Gradient, sortPriority=1}) {name=Main Color, nodeColor=<rgba 0 1 0.8359 1>}
  NamedRerouteDeclaration(None=@n4.R) {name=Normal 1 R, nodeColor=<rgba 1 0 0.7969 1>}
  n22 = VectorParameter() {group=Caustics, parameterName=Caustics Color, sortPriority=2, defaultValue=<rgba 0.0132 0.2787 0.3177 1>}
  NamedRerouteDeclaration(None=@n22.RGB) {name=Caustics Color, nodeColor=<rgba 0.6413 0.8456 1 1>}
  NamedRerouteDeclaration(None=@n6.R) {name=Normal 2 R, nodeColor=<rgba 0 1 0.8828 1>}
  n23 = Multiply(
   A=Multiply(A=NamedRerouteUsage(), B=NamedRerouteUsage())
   B=ScalarParameter() {defaultValue=5, group=Sparkles, parameterName=Sparkle Multiply, sortPriority=2}
  ) {constB=5}
  n24 = ScalarParameter() {defaultValue=50, group=Sparkles, parameterName=Sparkle Power, sortPriority=3}
  NamedRerouteDeclaration(None=MaterialFunctionCall(Emissive=Power(Base=@n23, Exp=@n24) {constExponent=75}) {materialFunction=&MF_FX_EmissiveAdjustments}.Result) {name=Sparkles, nodeColor=<rgba 0 0.8125 1 1>}
  n25 = MaterialFunctionCall(Float3=MaterialFunctionCall(BaseNormal=@n4.RGB, AdditionalNormal=@n6.RGB) {materialFunction=&BlendAngleCorrectedNormals}.Result) {materialFunction=&BreakOutFloat3Components}
  n26 = ScalarParameter() {defaultValue=0.02, group=Normal, parameterName=Normal Intensity}
  NamedRerouteDeclaration(None=MaterialFunctionCall(R=Multiply(A=@n25.R, B=@n26), G=Multiply(A=@n25.G, B=@n26), B=@n25.B) {materialFunction=&AppendMany}.RGB) {name=Normals, nodeColor=<rgba 0 1 0.9297 1>}
  NamedRerouteUsage()
  n27 = NamedRerouteUsage()
  n28 = TextureSampleParameterCube!(UVs=ReflectionVectorWS(), Apply View MipBias) {
   group=Cube Map
   parameterName=ReflectionCubeMap
   samplerType=SAMPLERTYPE_LinearColor
   texture=/SparksMaps/Environment/Venue_A/S7/Sets/Floor_Set/Textures/T_ZenPoolReflection_A
  }
  n29 = LinearInterpolate(A=@n27, B=@n28.RGB, Alpha=ScalarParameter() {defaultValue=0.2, group=Cube Map, parameterName=ReflectionCubeMapBlend})
  QualitySwitch(Default=@n29, Low=@n29, High=@n27, Medium=@n29, Epic=@n27)
  n30 = Constant() {r=1}
  n31 = Constant()
  QualitySwitch(Default=@n30, Low=@n31, High=@n30, Medium=@n31, Epic=@n30)
  QualitySwitch(Default=@n30, Low=@n31, High=@n30, Medium=@n31, Epic=@n30)
  n32 = NamedRerouteUsage()
  n33 = NamedRerouteUsage()
  n34 = Multiply(A=@n33, B=NamedRerouteUsage())
  n35 = StaticSwitchParameter(
   True=Add(A=@n32, B=Multiply(A=@n33, B=@n34))
   False=StaticSwitchParameter(True=@n32, False=Constant()) {group=Flat Color, parameterName=Use Emissive for Flat Color, sortPriority=2}
  ) {group=Caustics, parameterName=Caustics, sortPriority=1}
  StaticSwitchParameter(True=Add(A=@n35, B=NamedRerouteUsage()), False=@n35) {group=Sparkles, parameterName=Use Sparkles, sortPriority=1}
 ]
}
