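# Additive, Nanite-enabled material that, by its name, fakes chromatic aberration on environment meshes using offset colored layers.
# Camera-to-object distance over a scale, floored by a Max, multiplies world position into CameraOffset, whose clamped result is added to noise.
# An optional static switch scales that divisor by one minus the audio MPC's BeatLifetime for beat reactivity.
# Custom primitive data selects one of three color parameters, defaulting to red, green and blue.
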Material /SparksMaps/Environment/Shared/Materials/M_ChromaticAberation_A
graph {
 properties {bAutomaticallySetUsageInEditor=false, bUsedWithNanite=true, blendMode=BLEND_Additive, usedShadingModels=Default Lit}
 nodes [
  n1 = WorldPosition()
  n2 = CameraPositionWS()
  n3 = ObjectPositionWS()
  n4 = ScalarParameter() {defaultValue=20000, parameterName=Scale Amount}
  n5 = Multiply(
   A=@n4
   B=OneMinus(None=CollectionParameter() {collection=/FMAudioReactiveData/AudioAnalysis/MPC_Sparks_Audio_Parameters, parameterName=BeatLifetime})
  )
  n6 = StaticBoolParameter() {parameterName=Use Beat Reactivity}
  n7 = Max(A=Divide(A=Distance(A=@n2, B=@n3), B=StaticSwitch(True=@n5, False=@n4, Value=@n6)))
  Add(
   A=Noise!(World Position, FilterWidth)
   B=MaterialFunctionCall(Offset Amount=ScalarParameter() {defaultValue=-2000, parameterName=PixelOffset}, World Position=Multiply(A=@n1.XYZ, B=@n7)) {materialFunction=/Engine/Functions/Engine_MaterialFunctions02/WorldPositionOffset/CameraOffset}.Clamped Camera Offset
  )
  n8 = ScalarParameter() {bUseCustomPrimitiveData=true, desc="Choose between the preset colors in the material (0, 1, 2)", parameterName="Color Choice (0,1,2)"}
  MaterialExpressionVectorParameter_0 = VectorParameter() {parameterName=Color A, defaultValue=<rgba 1 0 0 0>}
  MaterialExpressionVectorParameter_1 = VectorParameter() {parameterName=Color B, defaultValue=<rgba 0 1 0 0>}
  MaterialExpressionVectorParameter_2 = VectorParameter() {parameterName=Color C, defaultValue=<rgba 0 0 1 0>}
  Switch!(
   SwitchValue=@n8
   Default=@MaterialExpressionVectorParameter_0.RGB
   Red=@MaterialExpressionVectorParameter_0.RGB
   Green=@MaterialExpressionVectorParameter_1.RGB
   Blue=@MaterialExpressionVectorParameter_2.RGB
  ) {inputs [
    {
     inputName=Red
     input {
      expression=&self:MaterialExpressionVectorParameter_0
      inputName=None
      mask=1
      maskA=0
      maskB=1
      maskG=1
      maskR=1
      outputIndex=0
     }
    }
    {
     inputName=Green
     input {
      expression=&self:MaterialExpressionVectorParameter_1
      inputName=None
      mask=1
      maskA=0
      maskB=1
      maskG=1
      maskR=1
      outputIndex=0
     }
    }
    {
     inputName=Blue
     input {
      expression=&self:MaterialExpressionVectorParameter_2
      inputName=None
      mask=1
      maskA=0
      maskB=1
      maskG=1
      maskR=1
      outputIndex=0
     }
    }
   ]}
 ]
}
