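# Default Lit, Nanite-enabled window material that fakes depth with chromatic channel separation, with optional audio-beat reactivity.
# One greyscale texture is sampled flat and at two BumpOffset depths, packed into RGB, with depth optionally scaled by inverted BeatLifetime.
# A static switch can replace this with the stage-show primary color (both gradient lerp inputs are identical).
# The color is tinted, darkened by subtraction, saturated and multiplied by emissive intensity.
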Material /SparksMaps/Environment/Shared/Materials/M_ChromaticWindow_A
graph {
 properties {bAutomaticallySetUsageInEditor=false, bUsedWithNanite=true, usedShadingModels=Default Lit}
 nodes [
  n1 = MaterialFunctionCall!(CustomUPosition) {materialFunction=/SparksStageShow/Systems/StageManager/FXManager/MaterialFunctions/MF_MojoStageColors}
  n2 = Power(Base=MaterialFunctionCall() {materialFunction=&LinearGradient}.VGradient)
  n3 = Multiply(A=TextureCoordinate(), B=ScalarParameter() {defaultValue=1, group=UVs, parameterName=UV Size, sortPriority=0})
  n4 = TextureObjectParameter() {parameterName=Greyscale Texture, samplerType=SAMPLERTYPE_Masks, sortPriority=0, texture=/Game/Global/Textures/Default/Blanks/T_MidgreyMask}
  n5 = TextureSample(UVs=@n3, Tex=@n4)
  n6 = ScalarParameter() {defaultValue=1, group=UVs, parameterName=Depth, sortPriority=1}
  n7 = CollectionParameter() {collection=/FMAudioReactiveData/AudioAnalysis/MPC_Sparks_Audio_Parameters, parameterName=BeatLifetime}
  n8 = StaticSwitch(
   True=Multiply(A=@n6, B=OneMinus(None=@n7))
   False=@n6
   Value=StaticBoolParameter() {group=Beat Reactivity, parameterName=Use Beat Reactivity, sortPriority=0}
  )
  n9 = Constant() {r=-1}
  n10 = TextureSample(UVs=BumpOffset(Coordinate=@n3, Height=Multiply(A=@n8, B=@n9)), Tex=@n4)
  n11 = Constant() {r=-2}
  n12 = TextureSample(UVs=BumpOffset(Coordinate=@n3, Height=Multiply(A=@n8, B=@n11)), Tex=@n4)
  n13 = Multiply(
   A=StaticSwitchParameter(
    True=LinearInterpolate(A=@n1.Primary, B=@n1.Primary, Alpha=@n2)
    False=MaterialFunctionCall(R=@n5.R, G=@n10.G, B=@n12.B) {materialFunction=&AppendMany}.RGB
   ) {parameterName=UseStageShowColor}
   B=VectorParameter() {parameterName=Tint, sortPriority=1, defaultValue=<rgba 1 1 1 1>}.RGB
  )
  n14 = ScalarParameter() {parameterName=Darken Ammount, sliderMax=1, sortPriority=2}
  Multiply(A=Saturate(None=Subtract(A=@n13, B=@n14) {constB=0.9}), B=ScalarParameter() {parameterName=Emissive Intensity, sortPriority=3})
  ScalarParameter() {parameterName=Metallic, sliderMax=1, sortPriority=4}
  ScalarParameter() {defaultValue=1, parameterName=Specular, sliderMax=1, sortPriority=5}
  ScalarParameter() {parameterName=Roughness, sliderMax=1, sortPriority=6}
 ]
}
