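# Nanite-enabled lit diamond material for Sparks stage sets, faking gem depth, reflection and glitter.
# Two bump-offset Voronoi normals blend, while a view-angle term drives a curve-atlas gradient.
# Mesh-UV and screen-aligned masks multiply into edge-weighted sparkles, mixed with a reflection cubemap for emissive, with optional PBR textures.
# Optional WPO rotates vertices about the object pivot by a sine or curve atlas, timed by clock or music measure.
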Material /SparksMaps/Environment/Shared/Materials/M_Diamond_A
graph {
 properties {bAutomaticallySetUsageInEditor=false, bUsedWithNanite=true, usedShadingModels=Default Lit}
 nodes [
  n1 = Multiply(A=NamedRerouteUsage(), B=NamedRerouteUsage())
  n2 = ScalarParameter() {defaultValue=1, group=Sparkles, parameterName=Sparkle Intensity}
  n3 = TextureSampleParameterCube!(UVs=ReflectionVectorWS(), Apply View MipBias) {
   group=Cube Map
   parameterName=ReflectionCubeMap
   samplerType=SAMPLERTYPE_LinearColor
   texture=/SparksMaps/Environment/S6/Venue_A/Sets/Diamond/Textures/T_DiamondCubeRT
  }
  n4 = LinearInterpolate(
   A=LinearInterpolate(A=Multiply(A=@n1, B=@n2), B=NamedRerouteUsage())
   B=@n3.RGB
   Alpha=ScalarParameter() {defaultValue=0.7, group=Cube Map, parameterName=Cube Map Intensity}
  ) {constAlpha=0.7}
  n5 = OneMinus(None=NamedRerouteUsage())
  MaterialFunctionCall(Emissive=StaticSwitch(True=Multiply(A=@n4, B=@n5), False=@n4, Value=NamedRerouteUsage())) {materialFunction=&MF_FX_EmissiveAdjustments}
  n6 = TextureSampleParameter2D(UVs=NamedRerouteUsage()) {
   group=Sparkles
   parameterName=Edge Texture
   samplerType=SAMPLERTYPE_Masks
   texture=/SparksMaps/Environment/S6/Venue_A/Sets/Diamond/Textures/T_DiamondPenta_VC
  }
  NamedRerouteDeclaration(None=@n6.R) {name=EdgeTexture, nodeColor=<rgba 0 1 0.8828 1>}
  n7 = TextureSampleParameter2D(UVs=Multiply(A=Constant() {r=1}, B=TextureCoordinate())) {group=Sparkles, parameterName=Sparkle Mask, texture=/SparksMaps/Environment/Shared/Materials/Generic/Textures/T_SparkleMask_A}
  n8 = TextureSampleParameter2D(UVs=Multiply(
    A=MaterialFunctionCall() {materialFunction=&ScreenAlignedUVs}."X 100%, Y 100%"
    B=Constant() {r=1}
   )) {group=Sparkles, parameterName=Sparkle Mask, texture=&T_SparkleMask_A}
  n9 = Multiply(A=Power(Base=Multiply(A=@n7.R, B=@n8.R))) {constB=5}
  n10 = Multiply(A=NamedRerouteUsage()) {constB=10}
  NamedRerouteDeclaration(None=StaticSwitchParameter(True=Multiply(A=@n9, B=@n10), False=@n9) {defaultValue=true, group=Sparkles, parameterName=_Use Edge Texture}) {name=Sparkles, nodeColor=<rgba 1 0.5625 0 1>}
  n11 = NamedRerouteUsage()
  n12 = TextureSampleParameter2D(UVs=@n11) {group=PBR Textures, parameterName=PBR Masks}
  NamedRerouteDeclaration(None=StaticSwitch(
    True=NamedRerouteDeclaration(None=@n12.R) {name=Specular Texture, nodeColor=<rgba 0.7344 1 0 1>}
    False=Constant() {r=1}
    Value=NamedRerouteUsage()
   )) {name=Final Specular, nodeColor=<rgba 0 1 0.625 1>}
  NamedRerouteDeclaration(None=StaticSwitch(
    True=NamedRerouteDeclaration(None=@n12.G) {name=Metallic Texture, nodeColor=<rgba 0.9453 0 1 1>}
    False=ScalarParameter() {defaultValue=1, parameterName=Metallic}
    Value=NamedRerouteUsage()
   )) {name=Final Metallic, nodeColor=<rgba 1 0.3047 0 1>}
  NamedRerouteDeclaration(None=StaticSwitch(
    True=NamedRerouteDeclaration(None=@n12.B) {name=Roughness Texture, nodeColor=<rgba 1 0 0.0937 1>}
    False=ScalarParameter() {parameterName=Rough}
    Value=NamedRerouteUsage()
   )) {name=Final Roughness, nodeColor=<rgba 0 0.0156 1 1>}
  n13 = TextureSampleParameter2D(UVs=BumpOffset(Coordinate=NamedRerouteUsage(), Height=ScalarParameter() {defaultValue=-100, group=Normal, parameterName=Normal Depth}) {heightRatio=0.005}) {
   group=Normal
   parameterName=Normal
   samplerType=SAMPLERTYPE_Normal
   texture=/SparksMaps/Environment/Shared/Materials/Generic/Textures/T_BasicVoronoi_A_N
  }
  NamedRerouteDeclaration(None=@n13.RGB) {name=Base Normal, nodeColor=<rgba 0.7578 1 0 1>}
  NamedRerouteDeclaration(None=@n13.R) {name=NormalR, nodeColor=<rgba 1 0 0.3984 1>}
  n14 = VertexNormalWS()
  n15 = TextureSampleParameter2D(UVs=BumpOffset(Coordinate=Fresnel(Normal=@n14) {baseReflectFraction=0.0001}, Height=Constant() {r=-10})) {
   group=Normal
   parameterName=Normal
   samplerType=SAMPLERTYPE_Normal
   texture=&T_BasicVoronoi_A_N
  }
  NamedRerouteDeclaration(None=MaterialFunctionCall(BaseNormal=@n13.RGB, AdditionalNormal=@n15.RGB) {materialFunction=&BlendAngleCorrectedNormals}.Result) {name=VoronoiNormal, nodeColor=<rgba 0 1 0.3203 1>}
  n16 = Power(Base=DotProduct(A=VertexNormalWS(), B=CameraVectorWS())) {constExponent=5}
  n17 = NamedRerouteUsage()
  n18 = CurveAtlasRowParameter(CurveTime=Saturate(None=Add(A=@n16, B=@n17))) {
   atlas=/SparksMaps/Environment/Shared/Materials/Generic/Textures/CurveAtlas/CA_Diamond
   curve=/SparksMaps/Environment/Shared/Materials/Generic/Textures/CurveAtlas/C_DiamondCurve_B
   defaultValue=1
   group=Gradient Curve
   parameterName=Gradient
  }
  NamedRerouteDeclaration(None=Add(A=@n16, B=@n18)) {name=Gradient, nodeColor=<rgba 0.6172 1 0 1>}
  NamedRerouteDeclaration(None=Multiply(A=TextureCoordinate(), B=ScalarParameter() {defaultValue=1, group=PBR Textures, parameterName=UV Tiling})) {name=UV Size, nodeColor=<rgba 0.5469 0 1 1>}
  NamedRerouteDeclaration(None=StaticBoolParameter() {group=PBR Textures, parameterName=_Use PBR Textures}) {name=Use PBR Textures, nodeColor=<rgba 0 1 0.2266 1>}
  NamedRerouteDeclaration(None=TextureSampleParameter2D(UVs=@n11) {group=PBR Textures, parameterName=Diffuse}.RGB) {name=Diffuse Texture, nodeColor=<rgba 0 0.4141 1 1>}
  NamedRerouteUsage()
  NamedRerouteUsage()
  NamedRerouteUsage()
  n19 = NamedRerouteUsage()
  n20 = NamedRerouteUsage()
  n21 = Add(A=NamedRerouteUsage(), B=@n3.RGB)
  Reroute(None=StaticSwitch(True=LinearInterpolate(A=@n21, B=@n19, Alpha=@n20), False=@n21, Value=NamedRerouteUsage()))
  n22 = Add(
   A=CollectionParameter() {collection=/FMAudioReactiveData/AudioAnalysis/MPC_Sparks_Audio_Parameters, parameterName=MeasureLifetime}
   B=ScalarParameter() {parameterName=Offset}
  )
  n23 = MaterialFunctionCall() {materialFunction=&ObjectPivotPoint}
  n24 = Add(A=ComponentMask(None=@n23.Object Pivot Location) {r=true}, B=ComponentMask(None=@n23.Mesh Particle Pivot Location) {g=true})
  n25 = Divide(A=Add(A=Time(), B=Divide(A=@n24) {constB=10}), B=ScalarParameter() {defaultValue=15, parameterName=Speed}) {constB=15}
  n26 = CurveAtlasRowParameter(CurveTime=StaticSwitchParameter(True=Frac(None=@n22), False=Frac(None=@n25)) {parameterName=Animate to the Beat}) {
   atlas=/SparksMaps/Environment/Shared/Materials/Generic/Textures/CurveAtlas/CA_VertAnim
   curve=/SparksMaps/Environment/Shared/Materials/Generic/Textures/CurveAtlas/C_VertAnim_A
   parameterName=Curve
  }
  n27 = Multiply(A=@n26.R, B=ScalarParameter() {defaultValue=1, parameterName=Movement Intensity})
  n28 = Multiply(A=Time(), B=ScalarParameter() {defaultValue=2, parameterName=Rotation Speed})
  n29 = MaterialFunctionCall() {materialFunction=&ObjectPivotPoint}
  n30 = MaterialFunctionCall(Color=@n29.Object Pivot Location) {materialFunction=&SplitComponents}
  n31 = Divide(A=Sine(None=Add(A=@n28, B=@n30.R)), B=ScalarParameter() {defaultValue=40, parameterName=Rotate Ammount, sliderMax=100}) {constB=40}
  n32 = Transform(None=VectorParameter() {parameterName=Rotation Axis, defaultValue=<rgba 1 0 0 0>}.RGB) {transformSourceType=TRANSFORMSOURCE_Local}
  StaticSwitchParameter(
   True=MaterialFunctionCall!(
    RotationAngle=StaticSwitchParameter(True=@n27, False=@n31) {parameterName=Use Texture for Movement Variation}
    NormalizedRotationAxis=Normalize(VectorInput=@n32)
    PivotPoint=@n29.Object Pivot Location
    Position
    Normal=NamedRerouteUsage()
    TangentU
    TangentV
   ) {materialFunction=/Game/Characters/Player/Male/Medium/Bodies/M_MED_Inferno/FX/Mat/RotateVertexAboutAxis}.WPO
   False=Constant()
  ) {parameterName=Use WPO}
  n33 = NamedRerouteUsage()
  StaticSwitch(True=LinearInterpolate(A=NamedRerouteUsage(), B=@n33, Alpha=@n5), False=@n33, Value=NamedRerouteUsage())
 ]
}
