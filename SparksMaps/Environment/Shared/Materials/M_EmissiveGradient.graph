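# Unlit environment material that draws an animated three-color emissive gradient broken up by a tiled dot pattern.
# Colors come from MF_MojoStageColors or manual parameters via a static switch, then blend along time-scrolled UVs.
# An optional Beat Reactive switch multiplies in a UV-varying pulse driven by BeatLifetime from the Sparks audio parameter collection.
# Named reroutes carry colors, lights and beat mask, with intensity scaling feeding MF_FX_EmissiveAdjustments.
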Material /SparksMaps/Environment/Shared/Materials/M_EmissiveGradient
graph {
 properties {bAutomaticallySetUsageInEditor=false, shadingModel=MSM_Unlit, usedShadingModels=Unlit}
 nodes [
  n1 = Multiply(A=NamedRerouteUsage(), B=NamedRerouteUsage())
  n2 = NamedRerouteUsage()
  MaterialFunctionCall(Emissive=Multiply(
    A=StaticSwitchParameter(True=@n1, False=@n2) {defaultValue=true, parameterName=Beat Reactive}
    B=ScalarParameter() {defaultValue=2, parameterName=Emissive Intensity}
   )) {materialFunction=&MF_FX_EmissiveAdjustments}
  n3 = MaterialFunctionCall!(CustomUPosition) {materialFunction=/SparksStageShow/Systems/StageManager/FXManager/MaterialFunctions/MF_MojoStageColors}
  NamedRerouteDeclaration(None=StaticSwitchParameter(True=@n3.Primary, False=VectorParameter() {group=Color, parameterName=Primary Color, sortPriority=0, defaultValue=<rgba 1 0 0 1>}.RGBA) {defaultValue=true, group=Color, parameterName=Use Stage Colors, sortPriority=3}) {name=Primary Color, nodeColor=<rgba 0.7578 0 1 1>}
  NamedRerouteDeclaration(None=StaticSwitchParameter(True=@n3.Secondary, False=VectorParameter() {group=Color, parameterName=Secondary Color, sortPriority=1, defaultValue=<rgba 0 1 0 1>}.RGBA) {defaultValue=true, group=Color, parameterName=Use Stage Colors, sortPriority=3}) {name=Secondary Color, nodeColor=<rgba 0 1 0.4375 1>}
  NamedRerouteDeclaration(None=StaticSwitchParameter(True=@n3.Tertiary, False=VectorParameter() {group=Color, parameterName=Tertiary, sortPriority=2, defaultValue=<rgba 0 0 1 1>}.RGBA) {defaultValue=true, group=Color, parameterName=Use Stage Colors, sortPriority=3}) {name=Tertiary Color, nodeColor=<rgba 1 0.1172 0 1>}
  n4 = Add(A=Sine(None=CollectionParameter() {collection=/FMAudioReactiveData/AudioAnalysis/MPC_Sparks_Audio_Parameters, parameterName=BeatLifetime})) {constB=2.5}
  n5 = TextureCoordinate()
  n6 = Frac(None=Multiply(A=ChannelMaskParameter(None=@n5) {maskChannel=Green, parameterName="UV Channel - Scroll Direction\\", defaultValue=<rgba 0 1 0 0>}.R) {constB=0.25})
  n7 = SmoothStep(Value=Multiply(A=Divide(A=@n4), B=Sine(None=@n6))) {constMax=0.8, constMin=0.2}
  NamedRerouteDeclaration(None=Saturate(None=Add(A=@n7) {constB=0.3})) {name=Emissive Beat, nodeColor=<rgba 1 0.8906 0 1>}
  n8 = NamedRerouteUsage()
  n9 = TextureCoordinate()
  n10 = Multiply(A=Time()) {constB=0.2}
  n11 = ChannelMaskParameter(None=Frac(None=Add(A=@n9, B=@n10))) {maskChannel=Green, parameterName=UV Channel, defaultValue=<rgba 0 1 0 0>}
  n12 = LinearInterpolate(A=NamedRerouteUsage(), B=NamedRerouteUsage(), Alpha=Saturate(None=SmoothStep(Value=@n11.R) {constMax=0.75, constMin=0.25}))
  n13 = OneMinus(None=Saturate(None=Cosine(None=@n11.R)))
  n14 = TextureCoordinate()
  n15 = ScalarParameter() {defaultValue=50, parameterName=Dot Size}
  n16 = Distance(A=Frac(None=Multiply(A=@n14, B=@n15) {constB=20}), B=Constant() {r=0.5})
  n17 = Add(A=SmoothStep(Value=OneMinus(None=@n16)) {constMin=0.6}) {constB=0.4}
  NamedRerouteDeclaration(None=Multiply(A=LinearInterpolate(A=@n8, B=@n12, Alpha=@n13), B=Saturate(None=@n17))) {name=Lights, nodeColor=<rgba 0 1 0.25 1>}
 ]
}
