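# Environment material for lights that plays a single-row or single-column flipbook texture as emissive, with optional sync to the music.
# A static switch drives the phase from either the audio collection's MeasureLifetime or scaled time, both offset using vertex red.
# The flipbook result is multiplied by a rerouted input and intensity, then passes through MF_MojoBeatPulse and MF_FX_EmissiveAdjustments.
# Stage colors feed a Color reroute; another switch picks vertex color or texture for diffuse.
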Material /SparksMaps/Environment/Shared/Materials/M_Flipbook_Lights
graph {
 properties {bAutomaticallySetUsageInEditor=false, usedShadingModels=Default Lit}
 nodes [
  Frac(None=Time())
  n1 = Constant() {r=1}
  n2 = NamedRerouteUsage()
  n3 = Add(
   A=Step(Y=VertexColor().R, X=Divide(A=@n1, B=@n2))
   B=CollectionParameter() {collection=/FMAudioReactiveData/AudioAnalysis/MPC_Sparks_Audio_Parameters, parameterName=MeasureLifetime}
  )
  n4 = VertexColor()
  n5 = Time()
  n6 = Frac(None=Divide(A=Add(A=@n4.R, B=@n5), B=ScalarParameter() {defaultValue=10, group=Flipbook, parameterName="Animation Speed (Time Divider)", sortPriority=5}) {constB=20})
  n7 = StaticBoolParameter() {defaultValue=true, group=Flipbook, parameterName=Sync to beat, sortPriority=4}
  n8 = Constant() {r=1}
  n9 = NamedRerouteDeclaration(None=ScalarParameter() {defaultValue=6, group=Flipbook, parameterName=Number of Frames, sortPriority=6}) {name=Number Of Frames, nodeColor=<rgba 0.4297 0 1 1>}
  n10 = StaticBoolParameter() {group=Flipbook, parameterName=Animate UVs Horizontally, sortPriority=7}
  n11 = TextureObjectParameter() {group=Flipbook, parameterName=Texture, sortPriority=1}
  n12 = TextureCoordinate()
  n13 = AppendVector(
   A=ScalarParameter() {defaultValue=1, group=Flipbook, parameterName=U Multiplier, sortPriority=2}
   B=ScalarParameter() {defaultValue=1, group=Flipbook, parameterName="V  Multiplier", sortPriority=3}
  )
  n14 = Multiply(
   A=NamedRerouteUsage()
   B=MaterialFunctionCall(
    "Animation  Phase (0-1)"=StaticSwitch(True=@n3, False=@n6, Value=@n7)
    Number of Rows=StaticSwitch(True=@n8, False=@n9, Value=@n10)
    Number of Columns=StaticSwitch(True=@n9, False=@n8, Value=@n10)
    Texture=NamedRerouteDeclaration(None=@n11) {name=Main Texture, nodeColor=<rgba 0 0.0625 1 1>}
    UVs=Multiply(A=@n12, B=@n13)
   ) {materialFunction=&FlipBook}.Result
  )
  n15 = ScalarParameter() {defaultValue=2, group=Emissive, parameterName=Emissive Intensity}
  MaterialFunctionCall(Emissive=MaterialFunctionCall!(Emissive=Multiply(A=@n14, B=@n15), BeatPulse Strength, BeatPulse Min, BeatPulse Max) {materialFunction=/SparksMaps/Environment/MaterialFunctions/MF_MojoBeatPulse}.Result) {materialFunction=&MF_FX_EmissiveAdjustments}
  n16 = MaterialFunctionCall!(CustomUPosition) {materialFunction=/SparksStageShow/Systems/StageManager/FXManager/MaterialFunctions/MF_MojoStageColors}
  Multiply(A=@n16.Secondary, B=@n16.Secondary)
  Multiply(A=@n16.Primary, B=@n16.Primary)
  NamedRerouteDeclaration(None=LinearInterpolate(
    A=@n16.Primary
    B=@n16.Primary
    Alpha=MaterialFunctionCall() {materialFunction=&LinearGradient}.UGradient
   )) {name=Color, nodeColor=<rgba 1 0.8438 0 1>}
  NamedRerouteUsage()
  StaticSwitch(
   True=VertexColor()
   False=TextureSample(Tex=NamedRerouteUsage()).RGB
   Value=StaticBoolParameter() {defaultValue=true, group=Diffuse, parameterName=Use Vertex Color for Diffuse, sortPriority=1}
  )
 ]
}
