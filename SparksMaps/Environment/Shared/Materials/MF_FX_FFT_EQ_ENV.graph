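# Material function that draws audio-equalizer masks as bars, smooth mountains and box grids, from an FFT texture or scrolling fake FFT noise.
# Linear or radial UVs are floor-quantized for bars and left continuous for mountains, with optional Perlin randomization of the lookup.
# Scaled, offset samples are thresholded against vertical UV by Step or hardness-controlled SmoothStep, and a sine mask sets bar width.
# The processed FFT values are also output.
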MaterialFunction /SparksMaps/Environment/Shared/Materials/MF_FX_FFT_EQ_ENV
graph {nodes [
  n1 = NamedRerouteUsage()
  n2 = FunctionInput(Preview=Constant()) {bUsePreviewValueAsDefault=true, inputName=FFT_Add, inputType=FunctionInput_Scalar, sortPriority=11}
  n3 = NamedRerouteUsage()
  n4 = Multiply(A=@n1, B=@n3)
  n5 = Multiply(A=Divide(A=Floor(None=@n4), B=@n3), B=Constant2Vector() {r=1})
  n6 = FunctionInput(Preview=StaticBool()) {
   bUsePreviewValueAsDefault=true
   description=Uses generic noise texture to randomize UV placement
   inputName=RandomizeUVPlacement
   inputType=FunctionInput_StaticBool
   sortPriority=1
  }
  n7 = StaticSwitch(
   True=TextureSample(UVs=@n5, Tex=NamedRerouteUsage()) {samplerType=SAMPLERTYPE_Masks, texture=/Game/Global/Textures/Default/Noise/T_Generic_Perlin}.R
   False=@n5
   Value=@n6
  )
  n8 = FunctionInput(Preview=StaticBool()) {
   bUsePreviewValueAsDefault=true
   description=Uses Fake FFT Noise MF
   inputName=Use Fake FFT Noise
   inputType=FunctionInput_StaticBool
   sortPriority=7
  }
  n9 = FunctionInput(Preview=TextureObject() {desc=Default Black 8x8, samplerType=SAMPLERTYPE_LinearColor, texture=/Game/Global/Textures/Default/Blanks/T_BlackLinearColor}) {bUsePreviewValueAsDefault=true, inputName=FFT, inputType=FunctionInput_Texture2D, sortPriority=6}
  n10 = StaticSwitch(
   True=MaterialFunctionCall!(
    In TexCoords=@n7
    Fake FFT Noise Texture
    Fake FFT Noise Size
    Fake FFT Noise Speed X=NamedRerouteUsage()
    Fake FFT Noise Speed Y
    Fake FFT Noise Intensity Mult
    Fake FFT Noise Pow=NamedRerouteUsage()
   ) {materialFunction=/Game/Effects/Niagara/MasterMaterials/Functions/Mojo_MF/MF_FX_Fake_FFT_Noise}.Result Horizontal
   False=TextureSample(UVs=@n7, Tex=@n9) {texture=/Engine/EngineResources/WhiteSquareTexture}.R
   Value=@n8
  )
  n11 = FunctionInput(Preview=Constant() {r=2}) {
   bUsePreviewValueAsDefault=true
   description=Fake FFT Controls Pow
   inputName=FFT_IntensityMult
   inputType=FunctionInput_Scalar
   sortPriority=10
  }
  n12 = NamedRerouteDeclaration(None=@n11) {name=FFTIntensityMult, nodeColor=<rgba 0.4531 1 0 1>}
  n13 = Saturate(None=Add(A=NamedRerouteDeclaration(None=@n2) {name=FFTAdd, nodeColor=<rgba 0 1 0.6484 1>}, B=Multiply(A=@n10, B=@n12) {constB=2}))
  n14 = Step(Y=ComponentMask(None=@n1) {g=true}, X=@n13)
  FunctionOutput(None=Multiply(A=NamedRerouteUsage(), B=@n14)) {outputName=EQ Bars}
  n15 = FunctionInput(Preview=Constant() {r=1}) {
   bUsePreviewValueAsDefault=true
   description=0 - 1 value controlling hardness
   inputName=MountainHardness
   inputType=FunctionInput_Scalar
   sortPriority=5
  }
  n16 = ComponentMask(None=NamedRerouteUsage()) {g=true}
  n17 = Multiply(A=NamedRerouteUsage(), B=Constant2Vector() {r=1})
  n18 = StaticSwitch(
   True=TextureSample(UVs=@n17, Tex=NamedRerouteUsage()) {samplerType=SAMPLERTYPE_Masks, texture=&T_Generic_Perlin}.R
   False=@n17
   Value=@n6
  )
  n19 = StaticSwitch(
   True=MaterialFunctionCall(In TexCoords=@n18, Fake FFT Noise Speed X=NamedRerouteUsage(), Fake FFT Noise Pow=NamedRerouteUsage()) {materialFunction=&MF_FX_Fake_FFT_Noise}.Result Horizontal
   False=TextureSample(UVs=@n18, Tex=@n9) {texture=&WhiteSquareTexture}.R
   Value=@n8
  )
  n20 = NamedRerouteUsage()
  n21 = Saturate(None=Add(A=Multiply(A=@n19, B=@n20) {constB=2}, B=NamedRerouteUsage()))
  n22 = SmoothStep(Min=@n16, Max=Add(A=@n16, B=OneMinus(None=@n15)), Value=@n21)
  FunctionOutput(None=@n22) {outputName=EQ Mountain, sortPriority=1}
  FunctionOutput(None=Multiply(A=NamedRerouteUsage(), B=@n22)) {outputName="EQ Mountain (Bar)", sortPriority=2}
  n23 = NamedRerouteUsage()
  FunctionOutput(None=Saturate(None=@n23)) {outputName=EQ Bar Boxes, sortPriority=3}
  FunctionOutput(None=@n13) {description=Sampled and processed FFT Mask Bar, outputName="FFT Sampled (Bar)", sortPriority=-2}
  FunctionOutput(None=@n21) {description=Sampled and processed FFT Mask Mountain, outputName="FFT Sampled (Mountain)", sortPriority=-1}
  n24 = FunctionInput(Preview=Constant() {desc=Width Control, r=0.25}) {bUsePreviewValueAsDefault=true, description=Thickness of Box Grid Lines, inputName=Square Grid Width, inputType=FunctionInput_Scalar}
  n25 = NamedRerouteUsage()
  n26 = NamedRerouteUsage()
  n27 = Frac(None=Multiply(A=NamedRerouteUsage(), B=AppendVector(A=@n25, B=@n26)))
  MaterialFunctionCall!(In TexCoords=@n27, Width=@n24) {materialFunction=/Game/Effects/Niagara/MasterMaterials/Functions/Mojo_MF/MF_FX_SquareEdgeMask}
  Multiply(A=@n23, B=NamedRerouteUsage())
  n28 = TextureCoordinate()
  n29 = ComponentMask(None=@n28) {r=true}
  n30 = OneMinus(None=ComponentMask(None=@n28) {g=true})
  n31 = FunctionInput(Preview=StaticBool()) {
   bUsePreviewValueAsDefault=true
   description="If using default UVs, this switches to a radial UV"
   inputName=RadialUV
   inputType=FunctionInput_StaticBool
   sortPriority=2
  }
  n32 = FunctionInput(Preview=StaticSwitch(
    True=MaterialFunctionCall() {materialFunction=&VectorToRadialValue}.Radial Coordinates
    False=AppendVector(A=@n29, B=@n30)
    Value=@n31
   )) {
   bUsePreviewValueAsDefault=true
   description=Defaults to using TexCoord or RadialUV - Can override here if something else is needed
   inputName=UV
   inputType=FunctionInput_Vector2
  }
  NamedRerouteDeclaration(None=@n32) {name=UV, nodeColor=<rgba 0.7578 0 1 1>}
  n33 = FunctionInput(Preview=Constant() {r=1}) {bUsePreviewValueAsDefault=true, description=EQ Bar Amount, inputName=BarAmount Vert, inputType=FunctionInput_Scalar, sortPriority=3}
  NamedRerouteDeclaration(None=@n33) {name=BarAmountVert, nodeColor=<rgba 1 0.5391 0 1>}
  NamedRerouteDeclaration(None=Saturate(None=@n27)) {name=BoxMask, nodeColor=<rgba 0 1 0.1094 1>}
  n34 = ComponentMask(None=NamedRerouteUsage()) {r=true}
  n35 = FunctionInput(Preview=Constant() {r=1}) {bUsePreviewValueAsDefault=true, description=EQ Bar Amount, inputName=BarAmount, inputType=FunctionInput_Scalar, sortPriority=3}
  n36 = NamedRerouteDeclaration(None=@n35) {name=BarAmount, nodeColor=<rgba 1 0.5391 0 1>}
  n37 = Sine(None=Add(A=Multiply(A=@n34, B=@n36) {constB=10}) {constB=0.2})
  n38 = FunctionInput(Preview=Constant() {r=1}) {bUsePreviewValueAsDefault=true, description=EQ Bar Width, inputName=BarWidth, inputType=FunctionInput_Scalar, sortPriority=4}
  NamedRerouteDeclaration(None=Step(Y=ConstantBiasScale(None=@n37), X=NamedRerouteDeclaration(None=@n38) {name=BarWidth, nodeColor=<rgba 1 0 0.2812 1>}) {constX=0.1}) {name=BarWidthSine, nodeColor=<rgba 0.9219 1 0 1>}
  NamedRerouteDeclaration(None=TextureObject() {samplerType=SAMPLERTYPE_Masks, texture=&T_Generic_Perlin}) {name=NoiseRandomize, nodeColor=<rgba 1 0 0.0703 1>}
  n39 = FunctionInput(Preview=Constant() {r=-0.139}) {
   bUsePreviewValueAsDefault=true
   description=Fake FFT Controls Speed
   inputName=Fake FFT Noise Speed X
   inputType=FunctionInput_Scalar
   sortPriority=8
  }
  NamedRerouteDeclaration(None=@n39) {name=FakeFFTSpeedX, nodeColor=<rgba 1 0 0.4688 1>}
  n40 = FunctionInput(Preview=Constant() {r=2}) {
   bUsePreviewValueAsDefault=true
   description=Fake FFT Controls Pow
   inputName=Fake FFT Noise Pow
   inputType=FunctionInput_Scalar
   sortPriority=9
  }
  NamedRerouteDeclaration(None=@n40) {name=FakeFFTPow, nodeColor=<rgba 0.9688 0 1 1>}
  NamedRerouteDeclaration(None=@n14) {name=FFTEQBarMask, nodeColor=<rgba 0.2422 1 0 1>}
 ]}
