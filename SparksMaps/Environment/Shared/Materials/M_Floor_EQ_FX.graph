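# Lit floor material for Sparks music stages that draws an audio-style equalizer as emissive bars, boxes and mountain shapes.
# MF_FX_FFT_EQ_ENV builds the EQ, with an audio FFT texture, fake FFT noise and MeasureBeat as signal sources.
# Named reroutes organize stepped gradients, grid masks and bump-offset UVs for parallax.
# A static switch picks stage-show colors or a fixed color, scaled by emissive intensity.
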Material /SparksMaps/Environment/Shared/Materials/M_Floor_EQ_FX
graph {
 properties {bAutomaticallySetUsageInEditor=false, usedShadingModels=Default Lit}
 nodes [
  n1 = MaterialFunctionCall!(CustomUPosition) {materialFunction=/SparksStageShow/Systems/StageManager/FXManager/MaterialFunctions/MF_MojoStageColors}
  n2 = ScalarParameter() {parameterName=StageShowColorSelect}
  n3 = StaticSwitchParameter(
   True=MaterialFunctionCall(A=@n1.Primary, B=@n1.Primary, C=@n1.Primary, Alpha=Saturate(None=@n2)) {materialFunction=&Lerp_3Color}.Result
   False=VectorParameter() {parameterName=Color, defaultValue=<rgba 1 0.0262 0.7011 1>}.RGB
  ) {defaultValue=true, parameterName=Use Stage Colors}
  Multiply(
   A=Multiply(A=NamedRerouteDeclaration(None=@n3) {name=Color, nodeColor=<rgba 0 0.1562 1 1>}, B=NamedRerouteUsage())
   B=ScalarParameter() {defaultValue=20, parameterName=Emissive Intensity}
  ) {constB=10}
  n4 = Step(Y=NamedRerouteUsage(), X=NamedRerouteUsage())
  Multiply(A=@n4, B=NamedRerouteUsage())
  NamedRerouteDeclaration(None=Constant() {r=0.1}) {name=GridWidth, nodeColor=<rgba 1 0.6563 0 1>}
  n5 = MaterialFunctionCall!(
   UV=NamedRerouteUsage()
   Square Grid Width=NamedRerouteUsage()
   RandomizeUVPlacement
   RadialUV
   BarAmount Vert=NamedRerouteUsage()
   BarAmount=NamedRerouteUsage()
   BarWidth=NamedRerouteUsage()
   MountainHardness
   FFT=NamedRerouteUsage()
   Use Fake FFT Noise=NamedRerouteUsage()
   Fake FFT Noise Speed X
   Fake FFT Noise Pow
   FFT_IntensityMult=NamedRerouteUsage()
   FFT_Add=Saturate(None=NamedRerouteUsage())
  ) {materialFunction=&/MF_FX_FFT_EQ_ENV}
  NamedRerouteDeclaration(None=@n5."FFT Sampled (Mountain)") {name=Mountain, nodeColor=<rgba 1 0.6094 0 1>}
  n6 = BumpOffset(Height=ScalarParameter() {defaultValue=-2}) {heightRatio=0.1}
  n7 = ComponentMask(None=@n6) {g=true}
  NamedRerouteDeclaration(None=AppendVector(A=ComponentMask(None=@n6) {r=true}, B=Subtract(A=@n7) {constB=0.3})) {name=UVBumpOffset, nodeColor=<rgba 0.6172 1 0 1>}
  n8 = OneMinus(None=@n4)
  NamedRerouteDeclaration(None=Saturate(None=Step(Y=@n8) {constX=0.5})) {name=Bar Stepped Mask, nodeColor=<rgba 1 0 0.4688 1>}
  NamedRerouteDeclaration(None=Constant() {desc=Num Bar Vert, r=4}) {name=BarAmountVert, nodeColor=<rgba 1 0.5391 0 1>}
  NamedRerouteDeclaration(None=@n5.EQ Bar Boxes) {name=EQBarBoxes, nodeColor=<rgba 0.1953 1 0 1>}
  NamedRerouteDeclaration(None=@n7) {name=VBumpOffset, nodeColor=<rgba 1 0.8906 0 1>}
  n9 = Saturate(None=NamedRerouteUsage())
  n10 = Subtract(A=OneMinus(None=NamedRerouteUsage()), B=@n9)
  n11 = Subtract(A=OneMinus(None=NamedRerouteUsage()), B=@n9)
  n12 = Saturate(None=Add(A=Saturate(None=@n10), B=Divide(A=@n11) {constB=3}))
  n13 = TextureCoordinate()
  n14 = Multiply(A=Subtract(A=ComponentMask(None=@n13) {g=true}) {constB=0.1}, B=Divide(A=NamedRerouteUsage()))
  n15 = OneMinus(None=NamedRerouteUsage())
  n16 = Add(A=Saturate(None=Add(A=@n12, B=@n14) {constB=0.1}), B=Multiply(A=Divide(A=@n15) {constB=5}, B=NamedRerouteUsage()))
  NamedRerouteDeclaration(None=Divide(A=Saturate(None=@n16))) {name=CombinedBars, nodeColor=<rgba 1 0.1641 0 1>}
  n17 = MaterialFunctionCall(
   UV=NamedRerouteUsage()
   Square Grid Width=NamedRerouteUsage()
   BarAmount Vert=NamedRerouteUsage()
   BarAmount=NamedRerouteUsage()
   BarWidth=NamedRerouteUsage()
   FFT=NamedRerouteUsage()
   Use Fake FFT Noise=NamedRerouteUsage()
   FFT_IntensityMult=NamedRerouteUsage()
  ) {materialFunction=&MF_FX_FFT_EQ_ENV}
  NamedRerouteDeclaration(None=@n17."FFT Sampled (Bar)") {name=FakeFFTSampledValue, nodeColor=<rgba 1 0 0.6562 1>}
  NamedRerouteDeclaration(None=@n5.EQ Mountain) {name=EQMountain, nodeColor=<rgba 1 0.7734 0 1>}
  NamedRerouteDeclaration(None=TextureObject() {texture=/FMAudioReactiveData/AudioAnalysis/T_Sparks_Audio_FFT}) {name=FFTTexObject, nodeColor=<rgba 0 1 0.2734 1>}
  n18 = MaterialFunctionCall!(
   In TexCoords=Multiply(A=TextureCoordinate(), B=Constant2Vector() {g=1})
   Fake FFT Noise Texture
   Fake FFT Noise Size
   Fake FFT Noise Speed X=Constant() {r=0.35}
   Fake FFT Noise Speed Y
   Fake FFT Noise Intensity Mult
   Fake FFT Noise Pow
  ) {materialFunction=/Game/Effects/Niagara/MasterMaterials/Functions/Mojo_MF/MF_FX_Fake_FFT_Noise}
  NamedRerouteDeclaration(None=LinearInterpolate(Alpha=Saturate(None=@n18.Result Horizontal)) {constA=0.15, constB=0.45}) {name=Intensity A, nodeColor=<rgba 1 0 0.0937 1>}
  n19 = CollectionParameter() {collection=/FMAudioReactiveData/AudioAnalysis/MPC_Sparks_Audio_Parameters, parameterName=MeasureBeat}
  NamedRerouteDeclaration(None=LinearInterpolate(Alpha=Saturate(None=@n19)) {constA=0.15, constB=0.45}) {name=Intensity B, nodeColor=<rgba 0 0.4141 1 1>}
  NamedRerouteDeclaration(None=StaticBool()) {name=UseFakeFFTNoise, nodeColor=<rgba 0 1 0.3906 1>}
  NamedRerouteDeclaration(None=Constant() {desc=Num Bar Width, r=3}) {name=BarAmount, nodeColor=<rgba 1 0.5391 0 1>}
  n20 = TextureCoordinate()
  n21 = OneMinus(None=ComponentMask(None=@n20) {g=true})
  NamedRerouteDeclaration(None=AppendVector(A=ComponentMask(None=@n20) {r=true}, B=Add(A=@n21) {constB=-0.2})) {name=UVs, nodeColor=<rgba 0.9688 0 1 1>}
  NamedRerouteDeclaration(None=Constant() {r=16}) {name=BarWidth, nodeColor=<rgba 0.3359 0 1 1>}
  NamedRerouteDeclaration(None=@n17.EQ Bars) {name=Bars1, nodeColor=<rgba 0 1 0.4375 1>}
  n22 = NamedRerouteUsage()
  n23 = NamedRerouteUsage()
  n24 = Multiply(A=Subtract(A=ComponentMask(None=@n22) {g=true}) {constB=0}, B=@n23)
  NamedRerouteDeclaration(None=Divide(A=Floor(None=@n24), B=@n23)) {name=SteppedVertGradient, nodeColor=<rgba 1 0 0.7031 1>}
  n25 = NamedRerouteUsage()
  n26 = NamedRerouteUsage()
  n27 = Frac(None=Multiply(A=NamedRerouteUsage(), B=AppendVector(A=@n25, B=@n26)))
  n28 = NamedRerouteUsage()
  NamedRerouteDeclaration(None=Saturate(None=MaterialFunctionCall!(In TexCoords=@n27, Width=@n28) {materialFunction=/Game/Effects/Niagara/MasterMaterials/Functions/Mojo_MF/MF_FX_SquareEdgeMask}.Mask One Minus)) {name=BoxMask, nodeColor=<rgba 0 1 0.1094 1>}
  NamedRerouteDeclaration(None=Saturate(None=@n5."FFT Sampled (Bar)")) {name=FFTSampledBar, nodeColor=<rgba 1 0.4219 0 1>}
  NamedRerouteUsage()
  n29 = TextureSampleParameter2D() {
   parameterName=Param
   samplerType=SAMPLERTYPE_Masks
   texture=/Game/Packages/Fortress_SharedMaterials/Textures/DetailTextures/Detail_GrimeSmallScratch
  }
  Power(Base=@n29.RGB) {constExponent=4}
  ScalarParameter() {parameterName=Metallic}
  ScalarParameter() {defaultValue=1, parameterName=Specular}
  ScalarParameter() {defaultValue=1.3, parameterName=Roughness}
  TextureSample() {samplerType=SAMPLERTYPE_Normal, texture=&FlatNormalBlue}
 ]
}
