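# Material function that selects UV channel 0, 1 or 2 through a scalar UVSet input and outputs it with tiling applied.
# Each channel's texture coordinate is multiplied by a vector appended from the Tile U and Tile V inputs.
# Nested If nodes compare against 1 and 2, with channel 0 (the Global UV reroute) apparently serving as the fallback for other values.
# Named reroutes distribute the inputs and shared UV through the graph.
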MaterialFunction /SparksMaps/Environment/Shared/Materials/Functions/MF_Choose_UV_Set
graph {nodes [
  n1 = NamedRerouteUsage()
  n2 = NamedRerouteUsage()
  n3 = NamedRerouteUsage()
  n4 = If(
   A=Constant() {r=1}
   B=NamedRerouteUsage()
   "A > B"=@n1
   "A == B"=Multiply(A=TextureCoordinate() {coordinateIndex=1}, B=AppendVector(A=@n2, B=@n3))
   "A < B"=@n1
  )
  n5 = Constant() {r=2}
  n6 = NamedRerouteUsage()
  n7 = NamedRerouteUsage()
  n8 = Multiply(A=TextureCoordinate() {coordinateIndex=2}, B=AppendVector(A=NamedRerouteUsage(), B=NamedRerouteUsage()))
  FunctionOutput(None=If(A=Constant() {r=1}, B=NamedRerouteUsage(), "A > B"=@n4, "A == B"=@n4, "A < B"=If(A=@n5, B=@n6, "A > B"=@n7, "A == B"=@n8, "A < B"=@n7)))
  n9 = FunctionInput() {bUsePreviewValueAsDefault=true, description=UV Set, inputName=Tile U, inputType=FunctionInput_Scalar, previewValue=<1 1 1 1>}
  NamedRerouteDeclaration(None=@n9) {name=TileU, nodeColor=<rgba 0 0.8125 1 1>}
  n10 = FunctionInput() {bUsePreviewValueAsDefault=true, description=UV Set, inputName=Tile V, inputType=FunctionInput_Scalar, previewValue=<1 1 1 1>}
  NamedRerouteDeclaration(None=@n10) {name=TileV, nodeColor=<rgba 1 0.8672 0 1>}
  n11 = FunctionInput() {bUsePreviewValueAsDefault=true, description=UV Set, inputName=UVSet, inputType=FunctionInput_Scalar, previewValue=<0 0 0 0>}
  NamedRerouteDeclaration(None=@n11) {name=UVSet, nodeColor=<rgba 0.5469 0 1 1>}
  n12 = NamedRerouteUsage()
  n13 = NamedRerouteUsage()
  NamedRerouteDeclaration(None=Multiply(A=TextureCoordinate(), B=AppendVector(A=@n12, B=@n13))) {name=Global UV, nodeColor=<rgba 0 1 0.2266 1>}
 ]}
