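# M_Generic_Heat_Emissive_A is an unlit, Nanite-enabled material for environment heat glow, with optional stage-show color and audio reactivity.
# A heat curve atlas supplies color, optionally desaturated and tinted by stage color or pulsed by music measure timing.
# Gradient masking and panning cloud noise shape the emissive, with optional Lumen boost varying by quality and shading path.
# Optional WPO rotates vertices around the pivot, driven by sine or an animation curve with position-based offsets.
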Material /SparksMaps/Environment/Shared/Materials/M_Generic_Heat_Emissive_A
graph {
 properties {
  bAutomaticallySetUsageInEditor=false
  bUsedWithNanite=true
  pixelDepthOffsetMode=PDOM_Legacy
  shadingModel=MSM_Unlit
  usedShadingModels=Unlit
 }
 nodes [
  n1 = MaterialFunctionCall(UV Channel=MaterialFunctionCall!(Tile U, Tile V, UVSet=Constant() {r=1}) {materialFunction=/SparksMaps/Environment/Shared/Materials/Functions/MF_Choose_UV_Set}.Result) {materialFunction=&LinearGradient}
  Multiply(A=Subtract(A=@n1.VGradient, B=Constant() {r=0.2}), B=@n1.VGradient)
  n2 = NamedRerouteUsage()
  n3 = CurveAtlasRowParameter(CurveTime=ComponentMask(None=TextureCoordinate()) {r=true}) {
   atlas=/SparksMaps/Environment/Shared/Materials/Generic/Textures/CurveAtlas/CA_HeatCurve
   curve=/SparksMaps/Environment/Shared/Materials/Generic/Textures/CurveAtlas/C_HeatCurve_A
   parameterName=HeatCurve
  }
  n4 = Constant() {r=1}
  n5 = StaticSwitchParameter(
   True=Multiply(
    A=MaterialFunctionCall!(CustomUPosition) {materialFunction=/SparksStageShow/Systems/StageManager/FXManager/MaterialFunctions/MF_MojoStageColors}.Primary
    B=Desaturation(None=@n3, Fraction=@n4)
   )
   False=@n3
  ) {parameterName=UseStageShowColor}
  NamedRerouteDeclaration(None=StaticSwitchParameter(True=Multiply(A=@n2, B=@n5), False=@n5) {parameterName=UseMeasureReactivity}) {name=Color, nodeColor=<rgba 0 0 0 1>}
  n6 = CurveAtlasRowParameter(CurveTime=CollectionParameter() {collection=/FMAudioReactiveData/AudioAnalysis/MPC_Sparks_Audio_Parameters, parameterName=MeasureLifetime}) {
   atlas=/SparksMaps/Effects/Screens/Curves/CurveAtlas_Audio_Beat_A
   curve=/SparksMaps/Effects/Screens/Curves/Curve_Audio_Beat_Remap_001
   parameterName=ReactivityCurve
  }
  NamedRerouteDeclaration(None=MaterialFunctionCall(Input=@n6, Input Low=Constant(), Input High=Constant() {r=1}, Target Low=Constant() {r=0.75}, Target High=Constant() {r=3}) {materialFunction=&RemapValueRange}.Result) {name=Reactivity, nodeColor=<rgba 0 0 0 1>}
  ScalarParameter() {parameterName=Opacity_UV_Set}
  n7 = MaterialFunctionCall() {materialFunction=&ObjectPivotPoint}
  n8 = Add(A=ComponentMask(None=@n7.Object Pivot Location) {r=true}, B=ComponentMask(None=@n7.Mesh Particle Pivot Location) {g=true})
  n9 = Add(A=Time(), B=Frac(None=Divide(A=@n8) {constB=10}))
  n10 = ScalarParameter() {defaultValue=15, parameterName=Speed}
  n11 = CurveAtlasRowParameter(CurveTime=Frac(None=Divide(A=@n9, B=@n10) {constB=15})) {
   atlas=/SparksMaps/Environment/Shared/Materials/Generic/Textures/CurveAtlas/CA_VertAnim
   curve=/SparksMaps/Environment/Shared/Materials/Generic/Textures/CurveAtlas/C_VertAnim_A
   parameterName=Curve
  }
  n12 = Multiply(A=@n11.R, B=ScalarParameter() {defaultValue=1, parameterName=Movement Intensity})
  n13 = Multiply(A=Time(), B=ScalarParameter() {defaultValue=2, parameterName=Rotation Speed})
  n14 = MaterialFunctionCall() {materialFunction=&ObjectPivotPoint}
  n15 = MaterialFunctionCall(Color=@n14.Object Pivot Location) {materialFunction=&SplitComponents}
  n16 = Divide(A=Sine(None=Add(A=@n13, B=@n15.R)), B=ScalarParameter() {defaultValue=40, parameterName=Rotation Ammount}) {constB=40}
  n17 = VectorParameter() {parameterName=Rotation Axis, defaultValue=<rgba 1 0 0 0>}
  StaticSwitchParameter(
   True=MaterialFunctionCall!(
    RotationAngle=StaticSwitchParameter(True=@n12, False=@n16) {parameterName=Use Curve for Movement Variation}
    NormalizedRotationAxis=Normalize(VectorInput=@n17.RGB)
    PivotPoint=@n14.Object Pivot Location
    Position
    Normal
    TangentU
    TangentV
   ) {materialFunction=/Game/Characters/Player/Male/Medium/Bodies/M_MED_Inferno/FX/Mat/RotateVertexAboutAxis}.WPO
   False=Constant()
  ) {parameterName=Use WPO}
  n18 = NamedRerouteUsage()
  n19 = ScalarParameter() {defaultValue=1, parameterName=Emissive Strength}
  n20 = MaterialFunctionCall() {materialFunction=&LinearGradient}
  n21 = Constant() {r=0.01}
  n22 = Multiply(
   A=Multiply(A=Multiply(A=@n18, B=@n19), B=@n20.UGradient)
   B=MaterialFunctionCall(In=Add(A=@n20.UGradient, B=@n21), Contrast=Constant()) {materialFunction=&CheapContrast}.Result
  )
  n23 = Panner(Speed=AppendVector(
    A=ScalarParameter() {defaultValue=0.1, parameterName=Fire_Cloud_Speed_Direction_Z}
    B=ScalarParameter() {parameterName=Fire_Cloud_Speed_Direction_X}
   )) {speedX=0.1}
  n24 = LinearInterpolate(
   A=Multiply(A=@n22, B=TextureSample(UVs=@n23) {texture=&AtmosphericCloudNoise01}.RGB)
   B=@n22
   Alpha=ScalarParameter() {defaultValue=0.6, parameterName=Noise Opacity}
  )
  n25 = Multiply(A=@n24, B=ScalarParameter() {defaultValue=3, group=LumenBoost, parameterName=EmissiveIntensity})
  n26 = RayTracingQualitySwitch(Normal=@n25, RayTraced=Multiply(A=@n25, B=ScalarParameter() {defaultValue=2, group=LumenBoost, parameterName=LumenBoost}))
  n27 = Multiply(A=@n24, B=ScalarParameter() {defaultValue=2, group=LumenBoost, parameterName=EmissiveIntensityLowEnd}) {constB=1.5, desc=Low End and Mobile Screen Multiplier}
  StaticSwitchParameter(True=ShadingPathSwitch(Default=QualitySwitch(Default=@n26, Low=@n27), Mobile=@n27), False=@n24) {group=LumenBoost, parameterName=UseLumenBoost}
 ]
}
