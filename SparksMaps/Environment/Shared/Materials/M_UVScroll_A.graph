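# M_UVScroll_A is a Default Lit environment material that continuously scrolls a single "Layer A" texture set.
# Time multiplied by U and V speed parameters offsets the texture coordinates, shared through a named reroute for diffuse, mask and normal sampling.
# The diffuse blends toward a version multiplied by tint color and alpha, using its own RGB as the blend factor.
# The mask texture is split into separate channels.
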Material /SparksMaps/Environment/Shared/Materials/M_UVScroll_A
graph {
 properties {bAutomaticallySetUsageInEditor=false, pixelDepthOffsetMode=PDOM_Legacy, usedShadingModels=Default Lit}
 nodes [
  n1 = TextureObjectParameter() {
   group=Blend A _Textures
   parameterName=Layer A Diffuse
   samplerType=SAMPLERTYPE_LinearColor
   texture=/SparksMaps/Environment/Green_Room/Textures/T_GR_PlasterWall_D
  }
  n2 = TextureSample(UVs=NamedRerouteUsage(), Tex=@n1)
  n3 = VectorParameter() {group=Blend A _Textures, parameterName="Tint (Layer A)"}
  LinearInterpolate(A=@n2.RGB, B=Multiply(A=Multiply(A=@n2.RGB, B=@n3.RGB), B=@n3.A), Alpha=@n2.RGB)
  n4 = TextureObjectParameter() {
   group=Blend A _Textures
   parameterName=Layer A Masks
   samplerType=SAMPLERTYPE_Masks
   texture=/Game/Global/Textures/Default/Blanks/T_MidgreyMask
  }
  MaterialFunctionCall(Float3=TextureSample(UVs=NamedRerouteUsage(), Tex=@n4).RGB) {materialFunction=&BreakOutFloat3Components}
  n5 = Time()
  n6 = Multiply(A=ScalarParameter() {parameterName=U Scroll Speed}, B=@n5)
  n7 = TextureCoordinate()
  n8 = ComponentMask(None=@n7) {r=true}
  n9 = ComponentMask(None=@n7) {g=true}
  n10 = Multiply(A=@n5, B=ScalarParameter() {parameterName=V Scroll Speed})
  NamedRerouteDeclaration(None=AppendVector(A=Add(A=@n6, B=@n8), B=Add(A=@n9, B=@n10))) {name=Layer A UVs, nodeColor=<rgba 0 0.6016 1 1>}
  n11 = TextureObjectParameter() {
   group=Blend A _Textures
   parameterName=Layer A Normal
   samplerType=SAMPLERTYPE_Normal
   texture=&FlatNormalBlue
  }
  TextureSample(UVs=NamedRerouteUsage(), Tex=@n11)
 ]
}
