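# Environment surface material for Sparks maps that swings meshes around their object pivot on a configurable axis using World Position Offset, rotating normals to match.
# Rotation comes from a pivot-offset time sine by default, or optionally a curve atlas driven by time or the audio MeasureLifetime parameter.
# A single texture layer supplies diffuse, normal and masks with optional tint.
# An optional masked emissive adds beat-pulsing stage-show colors.
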Material /SparksMaps/Environment/Shared/Materials/M_VertexRotate_Sine_A
graph {
 properties {bAutomaticallySetUsageInEditor=false, pixelDepthOffsetMode=PDOM_Legacy, usedShadingModels=Default Lit}
 nodes [
  n1 = TextureObjectParameter() {
   group=Blend A _Textures
   parameterName=Layer A Masks
   samplerType=SAMPLERTYPE_Masks
   texture=/Game/Global/Textures/Default/Blanks/T_MidgreyMask
  }
  MaterialFunctionCall(Float3=TextureSample(UVs=NamedRerouteUsage(), Tex=@n1).RGB) {materialFunction=&BreakOutFloat3Components}
  NamedRerouteDeclaration(None=TextureCoordinate()) {name=Layer A UVs, nodeColor=<rgba 0 0.6016 1 1>}
  n2 = Add(
   A=CollectionParameter() {collection=/FMAudioReactiveData/AudioAnalysis/MPC_Sparks_Audio_Parameters, parameterName=MeasureLifetime}
   B=ScalarParameter() {parameterName=Offset}
  )
  n3 = MaterialFunctionCall() {materialFunction=&ObjectPivotPoint}
  n4 = Add(A=ComponentMask(None=@n3.Object Pivot Location) {r=true}, B=ComponentMask(None=@n3.Mesh Particle Pivot Location) {g=true})
  n5 = Divide(A=Add(A=Time(), B=Divide(A=@n4) {constB=10}), B=ScalarParameter() {defaultValue=15, parameterName=Speed}) {constB=15}
  n6 = CurveAtlasRowParameter(CurveTime=StaticSwitchParameter(True=Frac(None=@n2), False=Frac(None=@n5)) {parameterName=Animate to the Beat}) {
   atlas=/SparksMaps/Environment/Shared/Materials/Generic/Textures/CurveAtlas/CA_VertAnim
   curve=/SparksMaps/Environment/Shared/Materials/Generic/Textures/CurveAtlas/C_VertAnim_A
   parameterName=Curve
  }
  n7 = ScalarParameter() {defaultValue=1, parameterName=Movement Intensity}
  n8 = Time()
  n9 = ScalarParameter() {defaultValue=2, parameterName=Rotation Speed}
  n10 = MaterialFunctionCall() {materialFunction=&ObjectPivotPoint}
  n11 = Sine(None=Add(
    A=Multiply(A=@n8, B=@n9)
    B=MaterialFunctionCall(Color=@n10.Object Pivot Location) {materialFunction=&SplitComponents}.R
   ))
  n12 = ScalarParameter() {defaultValue=40, parameterName=Rotate Ammount, sliderMax=100}
  n13 = VectorParameter() {parameterName=Rotation Axis, defaultValue=<rgba 1 0 0 0>}
  n14 = MaterialFunctionCall!(
   RotationAngle=StaticSwitchParameter(True=Multiply(A=@n6.R, B=@n7), False=Divide(A=@n11, B=@n12) {constB=40}) {parameterName=Use Texture for Movement Variation}
   NormalizedRotationAxis=Normalize(VectorInput=Transform(None=@n13.RGB) {transformSourceType=TRANSFORMSOURCE_Local})
   PivotPoint=@n10.Object Pivot Location
   Position
   Normal=NamedRerouteUsage()
   TangentU
   TangentV
  ) {materialFunction=/Game/Characters/Player/Male/Medium/Bodies/M_MED_Inferno/FX/Mat/RotateVertexAboutAxis}
  StaticSwitchParameter(True=@n14.WPO, False=Constant()) {defaultValue=true, parameterName=Use WPO}
  n15 = NamedRerouteUsage()
  n16 = TextureObjectParameter() {
   group=Blend A _Textures
   parameterName=Layer A Normal
   samplerType=SAMPLERTYPE_Normal
   texture=&FlatNormalBlue
  }
  StaticSwitchParameter(True=@n14.Normal, False=NamedRerouteDeclaration(None=TextureSample(UVs=@n15, Tex=@n16).RGB) {name=Normal, nodeColor=<rgba 0 1 0.25 1>}) {defaultValue=true, parameterName=Use WPO}
  n17 = TextureObjectParameter() {
   group=Blend A _Textures
   parameterName=Layer A Diffuse
   samplerType=SAMPLERTYPE_LinearColor
   texture=/SparksMaps/Environment/Green_Room/Textures/T_GR_PlasterWall_D
  }
  n18 = TextureSample(UVs=NamedRerouteUsage(), Tex=@n17)
  n19 = VectorParameter() {group=Blend A _Textures, parameterName="Tint (Layer A)", defaultValue=<rgba 1 1 1 1>}
  n20 = Multiply(A=@n18.RGB, B=@n19.RGB)
  StaticSwitchParameter(True=LinearInterpolate(A=@n18.RGB, B=Multiply(A=@n20, B=@n19.A), Alpha=@n18.RGB), False=@n18.RGB) {parameterName=Enable Tint}
  n21 = Multiply(
   A=TextureSample() {samplerType=SAMPLERTYPE_Masks, texture=/SparksMaps/Environment/S6/Venue_A/Sets/Lowrider/Textures/T_Lowrider_FX}.R
   B=MaterialFunctionCall!(
    Emissive=MaterialFunctionCall!(CustomUPosition) {materialFunction=/SparksStageShow/Systems/StageManager/FXManager/MaterialFunctions/MF_MojoStageColors}.Primary
    BeatPulse Strength
    BeatPulse Min
    BeatPulse Max
   ) {materialFunction=/SparksMaps/Environment/MaterialFunctions/MF_MojoBeatPulse}.Result
  )
  StaticSwitchParameter(
   True=MaterialFunctionCall(Emissive=Multiply(A=@n21) {constB=2}) {materialFunction=&MF_FX_EmissiveAdjustments}.Result
   False=Constant()
  ) {parameterName=Use Stage Show Color Mask}
 ]
}
