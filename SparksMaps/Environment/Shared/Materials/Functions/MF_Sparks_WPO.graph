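# Material function for Sparks environment materials that outputs an animated, noise-driven world position offset.
# It samples a blurred noise texture by world-aligned Z projection, panning it with time, then scales it by movement intensity.
# A static switch weights the motion by either a chosen vertex color channel or normalized height above the object pivot.
# Further static switches zero individual X, Y, Z axes or disable the offset entirely.
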MaterialFunction /SparksMaps/Environment/Shared/Materials/Functions/MF_Sparks_WPO
graph {nodes [
  n1 = ChannelMaskParameter(None=VertexColor()) {group=WPO, parameterName=Vertex Channel, sortPriority=10}
  n2 = TextureObjectParameter() {
   group=WPO
   parameterName=Noise Texture
   samplerType=SAMPLERTYPE_Grayscale
   sortPriority=5
   texture=&LowResBlurredNoise
  }
  n3 = WorldPosition()
  n4 = Multiply(A=Time(), B=ScalarParameter() {defaultValue=250, group=WPO, parameterName=WPO Speed, sortPriority=7}) {constB=250}
  n5 = Multiply(
   A=MaterialFunctionCall(
    TextureObject=@n2
    TextureSize=ScalarParameter() {defaultValue=100000, group=WPO, parameterName=Noise Size, sortPriority=6}
    WorldPosition=Add(A=@n3.XYZ, B=@n4)
   ) {materialFunction=&WorldAlignedTexture}.Z Texture
   B=ScalarParameter() {defaultValue=2500, group=WPO, parameterName=Movement Intensity, sortPriority=8}
  ) {constB=5000}
  n6 = MaterialFunctionCall() {materialFunction=&ObjectPivotPoint}
  n7 = Divide(A=Subtract(A=ComponentMask(None=@n3.XYZ) {b=true}, B=ComponentMask(None=@n6.Object Pivot Location) {b=true})) {constB=20000}
  n8 = MaterialFunctionCall(Color=StaticSwitch(
    True=Multiply(A=@n1.R, B=@n5)
    False=Multiply(A=@n7, B=@n5)
    Value=StaticBoolParameter() {group=WPO, parameterName=Use Vertex Color for WPO, sortPriority=9}
   )) {materialFunction=&SplitComponents}
  n9 = Constant()
  n10 = StaticSwitchParameter(True=@n8.R, False=@n9) {defaultValue=true, group=WPO, parameterName=Use X, sortPriority=2}
  n11 = StaticSwitchParameter(True=@n8.G, False=@n9) {defaultValue=true, group=WPO, parameterName=Use Y, sortPriority=3}
  n12 = StaticSwitchParameter(True=@n8.B, False=@n9) {defaultValue=true, group=WPO, parameterName=Use Z, sortPriority=4}
  FunctionOutput(None=StaticSwitch(
    True=MaterialFunctionCall(R=@n10, G=@n11, B=@n12) {materialFunction=&AppendMany}.RGB
    False=Constant()
    Value=StaticBoolParameter() {group=WPO, parameterName=Use WPO, sortPriority=1}
   ))
 ]}
